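# Translucent, unlit flame material for Niagara mesh particles on the Arcana Agate pickaxe.
# A gradient or sphere mask minus panned noise drives curve-atlas color, a distorted silhouette tint and opacity with edge, view and depth fades.
# A dynamic-parameter toggle blends to variant colors, and a static Glow switch substitutes flat glow color.
# UV-weighted world position offset adds sine sway, gravity and clamped drag, whose target uses Position X for all axes.
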Material /BRCosmetics/Weapons/FORT_Melee/Pickaxe_ArcanaAgate/FX/M_Pickaxe_ArcanaAgate_Flame
graph {
 properties {
  bAutomaticallySetUsageInEditor=false
  bUsedWithNiagaraMeshParticles=true
  blendMode=BLEND_Translucent
  shadingModel=MSM_Unlit
  twoSided=true
  usedShadingModels=Unlit
 }
 nodes [
  n1 = VectorParameter() {parameterName=Glow Color}
  n2 = VectorParameter() {parameterName=Glow Color Variant}
  n3 = DynamicParameter() {parameterIndex=1, defaultValue=<rgba 1 1 1 0>, paramNames [Position X, Position Y, Position Z, "Toggle (0-1)"]}
  n4 = ComponentMask(None=TextureCoordinate()) {g=true}
  n5 = OneMinus(None=@n4)
  n6 = Subtract(A=Power(Base=StaticSwitchParameter(True=@n5, False=@n4) {parameterName=GradientOrSphere?})) {constB=0.5}
  n7 = ScalarParameter() {group=Value, parameterName=Value Bias}
  n8 = TextureCoordinate()
  n9 = TextureCoordinate()
  n10 = ComponentMask(None=@n9) {r=true}
  n11 = Power(
   Base=ComponentMask(None=@n9) {g=true}
   Exp=ScalarParameter() {defaultValue=1, group=Gradient Shape, parameterName=UV Shape Stretch, sliderMax=5}
  )
  n12 = Constant() {r=0.5}
  n13 = ScalarParameter() {defaultValue=0.5, group=Gradient Shape, parameterName=Center Y, sliderMax=1}
  n14 = ScalarParameter() {group=Gradient Shape, parameterName=Hardness, sliderMax=1, sliderMin=-5}
  n15 = ComponentMask(None=TextureCoordinate()) {g=true}
  n16 = StaticSwitchParameter(
   True=OneMinus(None=ComponentMask(None=@n8) {g=true})
   False=SphereMask(A=AppendVector(A=@n10, B=@n11), B=AppendVector(A=@n12, B=@n13), Radius=Constant() {r=0.5}, Hardness=Multiply(A=@n14, B=@n15)) {hardnessPercent=0}
  ) {parameterName=GradientOrSphere?}
  n17 = ScalarParameter() {defaultValue=1, group=Gradient Shape, parameterName=Shape Multiply, sliderMax=5}
  n18 = OneMinus(None=ComponentMask(None=TextureCoordinate()) {g=true})
  n19 = Constant() {r=1}
  n20 = TextureCoordinate()
  n21 = ScalarParameter() {group=Dissolve, parameterName=UV Min, sliderMax=1}
  n22 = Subtract(A=ComponentMask(None=@n20) {g=true}, B=@n21)
  n23 = Subtract(A=ScalarParameter() {defaultValue=1, group=Dissolve, parameterName=UV Max, sliderMax=1}, B=@n21)
  n24 = Max(A=Min(A=Divide(A=@n22, B=@n23))) {constB=0}
  n25 = ScalarParameter() {defaultValue=1, group=Dissolve, parameterName=UV Stretch, sliderMax=5}
  n26 = Multiply(
   A=AppendVector(A=ComponentMask(None=@n20) {r=true}, B=Power(Base=@n24, Exp=@n25))
   B=AppendVector(
    A=ScalarParameter() {defaultValue=1, group=Dissolve, parameterName=Scale X, sliderMax=5}
    B=ScalarParameter() {defaultValue=1, group=Dissolve, parameterName=Scale Y, sliderMax=5}
   )
  )
  n27 = DynamicParameter() {paramNames [RandOffset,Time,Param3,Param4]}
  n28 = AppendVector(
   A=ScalarParameter() {group=Dissolve, parameterName=Speed X, sliderMax=5, sliderMin=-5}
   B=ScalarParameter() {defaultValue=0.5, group=Dissolve, parameterName=Speed Y, sliderMax=5, sliderMin=-5}
  )
  n29 = TextureSampleParameter2D(UVs=Add(A=Panner(Coordinate=@n26, Time=@n27.Time, Speed=@n28), B=AppendVector(A=@n27.RandOffset, B=@n27.RandOffset))) {
   group=Dissolve
   parameterName=Noise Texture
   samplerType=SAMPLERTYPE_Grayscale
   texture=/Game/Effects/Fort_Effects/Textures/Fire/T_WispyFireNoise_Soft_M
  }
  n30 = Subtract(
   A=Power(
    Base=Multiply(A=@n16, B=@n17) {constB=2}
    Exp=ScalarParameter() {defaultValue=1, group=Gradient Shape, parameterName=Shape Power, sliderMax=5}
   )
   B=Multiply(A=StaticSwitchParameter(True=@n18, False=@n19) {parameterName=Noise Graident}, B=@n29.R) {constA=1}
  )
  n31 = Power(
   Base=Add(A=Multiply(A=@n6, B=@n7) {constB=0.1}, B=@n30)
   Exp=ScalarParameter() {defaultValue=1, group=Value, parameterName=Value Power, sliderMax=5, sliderMin=1}
  )
  n32 = CurveAtlasRowParameter(CurveTime=@n31) {
   atlas=&/CA_ArcanaAgate_Flame
   curve=&/C_LinearColor_ArcanaAgate_Flame
  }
  n33 = CurveAtlasRowParameter(CurveTime=@n31) {
   atlas=&CA_ArcanaAgate_Flame
   curve=&/C_LinearColor_ArcanaAgate_Flame_Variant
   defaultValue=1
   parameterName=Param_Variant
  }
  n34 = LinearInterpolate(A=@n32, B=@n33, Alpha=@n3."Toggle (0-1)")
  n35 = VectorParameter() {parameterName=Silhouette_Color, defaultValue=<rgba 1 1 1 1>}
  n36 = VectorParameter() {parameterName=Silhouette_Color_Variant, defaultValue=<rgba 1 1 1 1>}
  n37 = AppendVector(A=Constant(), B=ScalarParameter() {parameterName=Silhouette_UV_Offset_Y})
  n38 = Saturate(None=@n30)
  n39 = ScalarParameter() {defaultValue=0.1, parameterName=Silhouette_Distortion}
  n40 = Add(A=TextureCoordinate(), B=AppendVector(A=Constant(), B=Multiply(A=@n38, B=@n39) {constB=0.1}))
  n41 = ScalarParameter() {defaultValue=1, parameterName=Silhouette_UV_Scale_X}
  n42 = ScalarParameter() {defaultValue=1, parameterName=Silhouette_UV_Scale_Y}
  n43 = TextureSampleParameter2D(UVs=MaterialFunctionCall(UVs=Add(A=@n37, B=@n40), Texture Scale=AppendVector(A=@n41, B=@n42)) {materialFunction=&ScaleUVsByCenter}.UVs) {
   parameterName=Silhouette Texture
   samplerType=SAMPLERTYPE_LinearGrayscale
   texture=&/T_ArcanaAgate_Pickaxe_FlameSilhouette
  }
  n44 = Multiply(A=ScalarParameter() {defaultValue=1, parameterName=Silhouette_Strength}, B=@n43.R)
  n45 = StaticSwitchParameter(
   True=LinearInterpolate(A=@n1.RGB, B=@n2.RGB, Alpha=@n3."Toggle (0-1)")
   False=LinearInterpolate(A=@n34, B=LinearInterpolate(A=@n35.RGB, B=@n36.RGB, Alpha=@n3."Toggle (0-1)"), Alpha=Saturate(None=@n44))
  ) {parameterName=Glow}
  n46 = ParticleColor()
  MaterialFunctionCall(Emissive=Multiply(A=@n45, B=@n46.RGB)) {materialFunction=&MF_FX_EmissiveAdjustments}
  MaterialFunctionCall(Hue Shift Percentage=@n34, Texture=@n45) {materialFunction=&HueShift}
  n47 = ScalarParameter() {group=Edge Control, parameterName=Hard Edge Start, sliderMax=1}
  n48 = Add(A=@n47, B=ScalarParameter() {defaultValue=1, group=Edge Control, parameterName=Hard Edge Fade, sliderMax=1})
  n49 = StaticSwitchParameter(True=SmoothStep(Min=@n47, Max=Saturate(None=@n48), Value=@n30), False=@n30) {defaultValue=true, parameterName=Hard Edge}
  n50 = StaticSwitchParameter(True=Multiply(A=LinearInterpolate(A=@n1.A, B=@n2.A, Alpha=@n3."Toggle (0-1)"), B=@n49), False=@n49) {parameterName=Glow}
  n51 = DotProduct(
   A=VertexNormalWS()
   B=MaterialFunctionCall() {materialFunction=&CameraDirectionVector}.Result
  )
  n52 = StaticSwitchParameter(
   True=SmoothStep(
    Min=ScalarParameter() {parameterName=ViewFadeEnd, sliderMax=1}
    Max=ScalarParameter() {defaultValue=0.5, parameterName=ViewFadeStart, sliderMax=1}
    Value=Abs(None=@n51)
   ) {constMax=0.5}
   False=Constant() {r=1}
  ) {parameterName=View Fade}
  Multiply(A=@n46.A, B=DepthFade(Opacity=Multiply(A=@n50, B=@n52), FadeDistance=ScalarParameter() {parameterName=FadeDistance}))
  n53 = Transform(None=VectorParameter() {group=WPO, parameterName=WPO Direction, defaultValue=<rgba 0 1 0 1>}.RGB) {transformSourceType=TRANSFORMSOURCE_Instance}
  n54 = ComponentMask(None=TextureCoordinate()) {g=true}
  n55 = Multiply(A=@n27.Time, B=ScalarParameter() {defaultValue=1, group=WPO, parameterName=WPO Speed, sliderMax=2})
  n56 = Multiply(
   A=Multiply(A=Normalize(VectorInput=@n53), B=ScalarParameter() {defaultValue=10, group=WPO, parameterName=WPO Strength, sliderMax=500})
   B=Sine(None=Add(A=@n54, B=@n55))
  )
  n57 = DynamicParameter() {parameterIndex=1, defaultValue=<rgba 1 1 1 0>, paramNames [Position X, Position Y, Position Z, "Toggle (0-1)"]}
  n58 = MaterialFunctionCall(X=@n57.Position X, Y=@n57.Position X, Z=@n57.Position X) {materialFunction=&MakeFloat3}
  n59 = ParticlePositionWS()
  n60 = VectorParameter() {group=WPO, parameterName=World Drag Mask, defaultValue=<rgba 1 1 1 1>}
  n61 = DynamicParameter() {parameterIndex=2, paramNames ["DragControl (0-1)", Param2, Param3, Param4]}
  n62 = ScalarParameter() {defaultValue=20, group=WPO, parameterName=World Drag Max, sliderMax=50}
  n63 = Clamp(
   None=Multiply(A=Subtract(A=@n58.Result, B=@n59), B=LinearInterpolate(A=@n60.RGB, Alpha=@n61."DragControl (0-1)"))
   Min=Multiply(A=@n62) {constB=-1}
   Max=@n62
  )
  n64 = ComponentMask(None=TextureCoordinate()) {g=true}
  n65 = OneMinus(None=@n64)
  Multiply(
   A=Add(A=Add(A=@n56, B=@n63), B=VectorParameter() {group=WPO, parameterName=Gravity WPO, defaultValue=<rgba 0 0 0 1>}.RGB)
   B=Power(
    Base=StaticSwitchParameter(True=@n64, False=@n65) {group=WPO, parameterName=WPOGradientFlip}
    Exp=ScalarParameter() {defaultValue=2, group=WPO, parameterName=WPO Bend}
   )
  )
  Time()
 ]
}
