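# Idle Niagara effect for the ArcanaAgate pickaxe, placing candle flames, glows and rising embers at three sockets.
# CandleFlame meshes feed smoothed world position, reactivity and swing state to dynamic material parameters, and change shape while swinging.
# Decal_Glow_And_Light adds glow sprites and a light; Embers drift upward with curl noise, tuned separately for front end and gameplay.
# System ReactivityIntensity boosts flames and embers, shifting glow and ember colors from orange to purple.
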NiagaraSystem /BRCosmetics/Weapons/FORT_Melee/Pickaxe_ArcanaAgate/FX/NS_Pickaxe_ArcanaAgate_Idle
dependencies {
 usedDataInterfaces [
  &NiagaraDataInterfaceArrayFloat3
  &NiagaraDataInterfaceCurve
  &NiagaraDataInterfaceColorCurve
  &NiagaraDataInterfaceSocketReader
 ]
 usedDynamicInputs [&SpawnGroup, &Divide_Float, &Subtract_Float]
 usedModules [
  &self:SystemSpawnScript.SystemScriptSource.SystemScriptGraph.NiagaraNodeAssignment_1.SetVariables_#1
  &self:SystemSpawnScript.SystemScriptSource.SystemScriptGraph.NiagaraNodeAssignment_4.SetVariables_#2
  &SystemState
  &self:SystemSpawnScript.SystemScriptSource.SystemScriptGraph.NiagaraNodeAssignment_5.SetVariables_#3
 ]
}
emitters {
 CandleFlame {
  emitterUpdateScript [
   &EmitterState(
    Inactive Response=::NewEnumerator1
    Life Cycle Mode=::NewEnumerator1
    Loop Behavior=::NewEnumerator1
    Loop Duration Mode=::NewEnumerator1
    UseLoopDelay=false
   )
   &SpawnBurst_Instantaneous(Spawn Count=3, Use Spawn Probability=false, UseLoopCountLimit=false)
  ]
  particleSpawnScript [
   &InitializeParticle(
    Color Mode=::NewEnumerator0
    Mesh Scale=&Multiply_VectorByFloat(Vector=<0.4 0.8 1>, Float=0.08)
    Mesh Scale Mode=::NewEnumerator4
    Sprite Size Mode=::NewEnumerator3
    Uniform Sprite Size=5
    UsePositionOffset=false
    Write Material Random=true
   )
   &InitialMeshOrientation(
    Mesh Axis To Orient=<0 0 1>
    Mesh Orientation Mode=::NewEnumerator2
    Orientation Vector=<0 1 0>
    Rotation=&SelectVectorFromArray(
     Direct Array Index=$Particles.PartitionParticles.Partition
     Vector Selection Array=DataInterface() {propertyValues {FloatData [<0 0 0>, <0 0 0>]}}
    )
    Use Rotation Vector=true
   )
   &CameraOffset()
   &FN_Mesh_Attributes(
    Create Orientation Attribute=true
    Create Scale Attribute=true
    Orientation:Quat4f=&FindQuatBetweenTwoVectors(From Vector=<1 0 0>, To Vector=<0 0 1>)
    Scale=&Multiply_VectorByFloat(
     Vector=&FN_FrontEndVector3Modifier(Gameplay=<1.92 0.8 0.8>, Front End=<1.28 0.8 0.8>)
     Float=0.025
    )
    Write to Module Namespace=true
   )
   /SetVariables(Particles.RandSpeed:NiagaraFloat=&RandomRangeFloat(Minimum=0, Maximum=2)) {isSetParameters=true}
   &DynamicMaterialParameters(
    Index 0 Param 1=&RandomRangeFloat(Minimum=-2, Maximum=2)
    Module.Index0Float0Write=true
    Module.Index0Float1Write=false
    Module.Index0Float2Write=false
    Module.Index0Float3Write=false
   ) {module=DynamicMaterialParameters001}
  ]
  particleUpdateScript [
   &ParticleState(Kill Particles When Lifetime Has Elapsed=false)
   &SolveForcesAndVelocity(Clamp Velocity=false, Limit Acceleration=false)
   &SocketLocation(
    SamplePosOffset=true
    Sampled Position Offset=&MakeVector(X=0, Y=0, Z=$System.ZOffset)
    Socket Index=&ReturnExecIndex()
    Socket Reader=$System.Socket Reader
    Socket Sampling Mode=::NewEnumerator3
    VelocitySpeedLimit=false
   )
   &InterpolateOverTime(
    Rate Of Change=5
    Target Vector Value:Vector3f=&TransformPosition(Position=$Particles.Position, Source Space=::Simulation, Destination Space=::World)
    Variable Type=::NewEnumerator2
   )
   /SetVariables(Particles.RandSpeed:NiagaraFloat=&Add_Float(
     A=$Particles.RandSpeed
     B=&Multiply_Float(A=$Engine.DeltaTime, B=&Lerp_Float(A=1, B=1.6, Alpha=$System.ReactivityIntensity))
    )) {isSetParameters=true}
   &DynamicMaterialParameters(
    Index 0 Param 2=$Particles.RandSpeed
    Index 1 Param 1=&MakeFloatFromVector(Vector=$Particles.InterpolateOverTime.InterpolateOverTime_V3_.Moving Average, Channel=::NewEnumerator0)
    Index 1 Param 2=&MakeFloatFromVector(Vector=$Particles.InterpolateOverTime.InterpolateOverTime_V3_.Moving Average, Channel=::NewEnumerator1)
    Index 1 Param 3=&MakeFloatFromVector(Vector=$Particles.InterpolateOverTime.InterpolateOverTime_V3_.Moving Average, Channel=::NewEnumerator2)
    Index 1 Param 4=&Lerp_Float(A=0, B=1, Alpha=$System.ReactivityIntensity)
    Index 2 Param 1=&MakeCustomFloatFromBool(Bool=$User.bisSwinging, True Float=0.85, False Float=0)
    Module.Index0Float0Write=false
    Module.Index0Float1Write=true
    Module.Index0Float2Write=false
    Module.Index0Float3Write=false
    Module.Index1Float0Write=true
    Module.Index1Float1Write=true
    Module.Index1Float2Write=true
    Module.Index1Float3Write=true
    Module.Index2Float0Write=true
    Module.Index2Float1Write=false
    Module.Index2Float2Write=false
    Module.Index2Float3Write=false
    Write Parameter Index 1=::NewEnumerator1
    Write Parameter Index 2=::NewEnumerator1
   )
   /SetVariables(Particles.Position02:NiagaraPosition=&AddVectorToPosition(
     Position=$Particles.Position
     Vector=&FN_FrontEndVector3Modifier(Gameplay=<0 0 0>, Front End=<0 0 0>)
    )) {isSetParameters=true}
   /SetVariables(
    Particles.FN_Mesh_Attributes.Scale:Vector3f=&Multiply_Vector(
     A=$Particles.Initial.FN_Mesh_Attributes.Scale
     B=&MakeVector(
      X=&MakeCustomFloatFromBool(Bool=$User.bisSwinging, True Float=0, False Float=&Lerp_Float(A=1, B=1.3, Alpha=$System.ReactivityIntensity))
      Y=1
      Z=1
     )
    )
    Particles.Scale=&Multiply_Vector(
     A=$Particles.Initial.Scale
     B=&MakeVector(
      X=1
      Y=&MakeCustomFloatFromBool(Bool=$User.bisSwinging, True Float=1.5, False Float=&Lerp_Float(A=1, B=1.3, Alpha=$System.ReactivityIntensity))
      Z=1
     )
    )
   ) {isSetParameters=true}
   &ScaleColor(
    Scale Alpha=&Lerp_Float(
     A=0
     B=1
     Alpha=&FloatFromCurve(
      FloatCurve=DataInterface() {propertyValues {Curve {keys [{interpMode=RCIM_Cubic, tangentMode=RCTM_User, time=0.2, value=0}, {interpMode=RCIM_Cubic, tangentMode=RCTM_User, time=0.8, value=1}]}}}
      CurveIndex=$Engine.System.Age
     )
    )
    Scale RGB=&VectorFromFloat(Value=&Lerp_Float(A=1, B=1.3, Alpha=$System.ReactivityIntensity))
    ScaleA=true
    ScaleRGB=true
   )
  ]
  rendererClasses [&NiagaraMeshRendererProperties]
  renderers [
   {rendererClass=&NiagaraMeshRendererProperties, rendererIndex=0}
   {rendererClass=&NiagaraMeshRendererProperties, rendererIndex=1}
   {rendererClass=&NiagaraMeshRendererProperties, rendererIndex=2}
  ]
 }
 Decal_Glow_And_Light {
  emitterUpdateScript [
   &EmitterState(
    Inactive Response=::NewEnumerator1
    Life Cycle Mode=::NewEnumerator1
    Loop Behavior=::NewEnumerator1
    Loop Duration Mode=::NewEnumerator1
    UseLoopDelay=false
   )
   &SpawnBurst_Instantaneous(Spawn Count=3, Use Spawn Probability=true) {module=SpawnBurst_Instantaneous001}
  ]
  particleSpawnScript [&InitializeParticle(
    Color Mode=::NewEnumerator0
    Lifetime=2
    Sprite Size Mode=::NewEnumerator3
    Uniform Sprite Size=12.5
    UsePositionOffset=false
    Write Material Random=true
   ) {module=InitializeParticle001}]
  particleUpdateScript [
   &ParticleState(Kill Particles When Lifetime Has Elapsed=false)
   &SocketLocation(
    SamplePosOffset=true
    Sampled Position Offset=&MakeVector(X=0, Y=0, Z=&Add_Float(A=$System.ZOffset, B=1))
    Socket Index=&ReturnExecIndex()
    Socket Reader=$System.Socket Reader
    Socket Sampling Mode=::NewEnumerator3
    VelocitySpeedLimit=false
   )
   &ScaleColor(
    Scale Alpha=&Lerp_Float(
     A=0
     B=1
     Alpha=&FloatFromCurve(
      FloatCurve=DataInterface() {propertyValues {Curve {keys [{interpMode=RCIM_Cubic, tangentMode=RCTM_User, time=0.2, value=0}, {interpMode=RCIM_Cubic, tangentMode=RCTM_User, time=0.8, value=1}]}}}
      CurveIndex=$Engine.System.Age
     )
    )
    Scale RGB=&MakeVectorFromLinearColorRGB(Color=&LerpLinearColors(StartColor=<rgba 3 1.1976 0 0.4>, EndColor=<rgba 1.0751 0.3 3 0.4>, LerpFactor=$System.ReactivityIntensity))
    ScaleA=true
    ScaleRGB=true
   )
   /SetVariables(Particles.LightColor=$Particles.Color) {isSetParameters=true}
   /SetVariables(Particles.SpriteSize=&Multiply_Vector2D_ByFloat(Float=1, Vector2D=$Particles.Initial.SpriteSize)) {isSetParameters=true}
   /SetVariables(Particles.LightRadius=&MakeCustomFloatFromBool(
     Bool=&SetBoolByFloatComparison(A=&ReturnNormalizedExecIndex(), B=0, Comparison Type=::NewEnumerator4)
     True Float=75
     False Float=0
    )) {isSetParameters=true}
  ]
  rendererClasses [&NiagaraLightRendererProperties, &NiagaraSpriteRendererProperties]
  renderers [
   {rendererClass=&NiagaraLightRendererProperties, rendererIndex=0}
   {rendererClass=&NiagaraSpriteRendererProperties, rendererIndex=1}
  ]
 }
 Embers {
  emitterUpdateScript [
   &EmitterState()
   &SpawnRate(
    Spawn Probability=&Lerp_Float(A=0.3, B=0.6, Alpha=$System.ReactivityIntensity)
    SpawnRate=&FN_FrontEndFloatModifier(Gameplay=15, Front End=15)
    Use Spawn Probability=true
   )
  ]
  particleSpawnScript [
   &FN_SetSpawnGroup()
   /SetVariables(
    Particles.PerParticleRandom:NiagaraFloat=&RandomRangeFloat(Minimum=0, Maximum=1)
    Particles.RandSocket:NiagaraInt32=&RandomRangeInt(Minimum=0, Maximum=2)
   ) {isSetParameters=true}
   &InitializeParticle(
    Color Mode=::NewEnumerator0
    Lifetime=&Lerp_Float(A=0.5, B=0.35, Alpha=$Particles.PerParticleRandom)
    Position Mode=::NewEnumerator0
    Sprite Size=&Multiply_Vector2D(
     A=&RandomRangeVector2D(Minimum=<0.25 0.8>, Maximum=<0.3 1>)
     B=&MakeVector2D(
      X=&FloatFromCurve(
       FloatCurve=DataInterface() {propertyValues {Curve {keys [<at 0 1>,<at 0.8 2 RCIM_Constant>,<at 0.95 3 RCIM_Constant>]}}}
       CurveIndex=$Particles.PerParticleRandom
       Scale Curve=0.75
      )
      Y=&FloatFromCurve(
       FloatCurve=DataInterface() {propertyValues {Curve {keys [<at 0 1>,<at 0.8 2 RCIM_Constant>,<at 0.95 3>]}}}
       CurveIndex=$Particles.PerParticleRandom
      )
     )
    )
    Sprite Size Mode=::NewEnumerator4
    Write Material Random=true
   )
   &SocketLocation(
    SamplePosOffset=true
    Sampled Position Offset=&MakeVector(X=0, Y=0, Z=&Add_Float(A=$System.ZOffset, B=3))
    Socket Index=$Particles.RandSocket
    Socket Reader=$System.Socket Reader
    Socket Sampling Mode=::NewEnumerator3
    VelocitySpeedLimit=false
   )
   &ShapeLocation(MaskSpawnGroup=false, Override Seed=false, Sphere Radius=3, UseOwnerScale=true)
   &AddVelocity(
    Cone Angle=35
    Cone Axis=&TransformVector(Vector=<0 0 1>, Destination Space=::World)
    Module.UseConeFalloff=false
    Module.UseVelDistribution=false
    Override Seed=false
    Rotation Coordinate Space=::Simulation
    Velocity Mode=::NewEnumerator2
    Velocity Speed=&Multiply_Float(
     A=&RandomRangeFloat(Minimum=25, Maximum=50)
     B=&Lerp_Float(A=1, B=1.4, Alpha=$System.ReactivityIntensity)
    )
   )
  ]
  particleUpdateScript [
   &ParticleState()
   &AccelerationForce(Acceleration=&FN_FrontEndVector3Modifier(Gameplay=<0 0 200>, Front End=<0 0 100>))
   &Drag(
    Drag=&FloatFromCurve(
     FloatCurve=DataInterface() {propertyValues {Curve {keys [{interpMode=RCIM_Cubic, tangentMode=RCTM_User, time=0.2, value=0}, {interpMode=RCIM_Cubic, tangentMode=RCTM_User, time=0.8, value=1}]}}}
     Scale Curve=&FN_FrontEndFloatModifier(
      Gameplay=&RandomRangeFloat(Minimum=10, Maximum=15)
      Front End=&RandomRangeFloat(Minimum=5, Maximum=10)
     )
    )
    Use Linear Drag=true
    Use Rotational Drag=false
   )
   /Niagara/Modules/Update/Forces/CurlNoiseForce(
    Noise Strength=&FloatFromCurve(
     FloatCurve=DataInterface() {propertyValues {Curve {keys [{tangentMode=RCTM_User, time=0, value=0}, {tangentMode=RCTM_User, time=1, value=1}]}}}
     Scale Curve=&RandomRangeFloat(Minimum=100, Maximum=200)
    )
    Pan Noise=true
    Pan Noise Field=<0 0 0.5>
   )
   &InheritSourceMovement(
    Applied Position Scale=&VectorFromFloat(Value=&FloatFromCurve(FloatCurve=DataInterface() {propertyValues {Curve {keys [{interpMode=RCIM_Cubic, tangentMode=RCTM_User, time=0, value=1}, <at 0.5 0 RCIM_Cubic>]}}}))
    ApplyForce=false
    ApplyPosition=true
    ApplyPositionFalloff=false
    Source Velocity:Vector3f=$Particles.SkeletalMeshLocation001.SampledVelocity
   )
   &SolveForcesAndVelocity(Clamp Velocity=false, Limit Acceleration=false)
   &ScaleColor(
    Scale Alpha=&FloatFromCurve(FloatCurve=DataInterface() {propertyValues {Curve {keys [<at 0 1 RCIM_Cubic>,<at 1 0 RCIM_Cubic>]}}})
    Scale RGB=&MakeVectorFromLinearColorRGB(Color=&LerpLinearColors(
      StartColor=&ColorFromCurve(
       ColorCurve=DataInterface() {propertyValues {
         AlphaCurve {keys [<at 0 1>,<at 1 1>]}
         BlueCurve {keys [<at 0.1802 0>,<at 0.4665 0.0622>,<at 0.7796 0.1812>,<at 1 0.8627>]}
         GreenCurve {keys [
           <at 0.1802 0.2381>
           {arriveTangent=0.1187, interpMode=RCIM_Cubic, leaveTangent=0.1187, tangentMode=RCTM_User, time=0.4665, value=0.5663}
           <at 0.7796 0.8388>
           {arriveTangent=6.9912, interpMode=RCIM_Cubic, leaveTangent=6.9912, tangentMode=RCTM_User, time=1, value=0.9911}
          ]}
         RedCurve {keys [<at 0.1802 0.6>,<at 0.4665 0.8>,<at 0.7796 0.9301>,<at 1 1>]}
        }}
       CurveIndex=&OneMinusFloat(Float=$Particles.NormalizedAge)
      )
      EndColor=&Multiply_LinearColorByFloat(
       Linear Color=&ColorFromCurve(
        ColorCurve=DataInterface() {propertyValues {
          AlphaCurve {keys [<at 0 1>,<at 1 1>]}
          BlueCurve {keys [<at 0.1802 0.6>,<at 0.4309 0.8>,<at 0.7796 0.93>,<at 1 1>]}
          GreenCurve {keys [
            <at 0.1802 0>
            <at 0.4309 0.0624>
            <at 0.7796 0.1813>
            {arriveTangent=6.9912, interpMode=RCIM_Cubic, leaveTangent=6.9912, tangentMode=RCTM_User, time=1, value=0.8627}
           ]}
          RedCurve {keys [<at 0.1802 0.1721>,<at 0.4309 0.2285>,<at 0.7796 0.5401>,<at 1 0.9911>]}
         }}
        CurveIndex=&OneMinusFloat(Float=$Particles.NormalizedAge)
       )
       Float=3
      )
      LerpFactor=$System.ReactivityIntensity
     ))
    ScaleA=true
    ScaleRGB=true
   )
   &ScaleSpriteSizeBySpeed(
    Max Scale Factor=<1 2>
    Sample Scale Factor By Curve=true
    Scale Factor Curve=DataInterface() {propertyValues {Curve {keys [<at 0 0>, {arriveTangent=1.4989, interpMode=RCIM_Cubic, tangentMode=RCTM_Break, time=1, value=1}]}}}
    Velocity Threshold=999
   )
   /SetVariables(Particles.Glow03_Color=&MultiplyLinearColor(
     Linear Color=$Particles.Color
     Scale Factor=&MakeFloat4FromVectorAndFloat(
      "Vector (RGB)"=&VectorFromFloat(Value=0.75)
      "Float (Alpha)"=&FN_FrontEndFloatModifier(Gameplay=2, Front End=0.9)
     )
    )) {isSetParameters=true}
   /SetVariables(Particles.Glow03_Size=&Multiply_Vector2D(A=$Particles.SpriteSize, B=<2.5 2>)) {isSetParameters=true}
  ]
  rendererClasses [&NiagaraSpriteRendererProperties]
  renderers [
   {rendererClass=&NiagaraSpriteRendererProperties}
   {rendererClass=&NiagaraSpriteRendererProperties, rendererIndex=1}
  ]
 }
}
summary {
 systemName=NS_Pickaxe_ArcanaAgate_Idle
 emitters [
  {emitterName=CandleFlame, rendererClasses [&NiagaraMeshRendererProperties]}
  {emitterName=Embers, rendererClasses [&NiagaraSpriteRendererProperties]}
  {
   emitterName=Decal_Glow_And_Light
   rendererClasses [&NiagaraLightRendererProperties, &NiagaraSpriteRendererProperties]
  }
 ]
 userVariables [
  User.bIsFrontEnd {type=NiagaraBool}
  User.bIsFrontEndPreview {type=NiagaraBool}
  User.bisSwinging {type=NiagaraBool}
  User.DebugReactivity {type=NiagaraBool}
  User.ReactivityTime {type=NiagaraFloat, defaultValue=0, description=""}
 ]
}
